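; Attached actor for the Lotus Scooter gameplay item that shows a mark effect to players on its team.
; It sets its lifespan from a Duration variable on begin play, with a custom event to reset that lifespan.
; The server copies the owner's team, waits briefly, then sets a replicated Spawned flag.
; The flag's rep notify unhides the mark particle on clients whose local player is on the same team.
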(event Custom|ResetLife
  ;[1 1 1] Reset Life
  (Actor|SetLifeSpan (Variables|Default|GetDuration)))

(event EventBeginPlay
  ;[1 1 1] Set Life then rep the spawned variable
  (Actor|SetLifeSpan (Variables|Default|GetDuration))
  (if (Networking|IsServer)
    (CallFunction|SetTeam (Team|GetActorTeam (Actor|GetOwner)))
    (Utilities|FlowControl|Delay 0.2)
    (Variables|Default|SetSpawned true)))

(fn OnRep_Spawned
  ;[1 1 1] If a friendly unhide the particle
  (if (Networking|IsServer)
    _
    (elif (Team|OnSameTeam self (Game|GetPlayerController 0))
      (Development|SetHiddeninGame (Variables|Default|GetMarkFX) false false))))

(fn ConstructionScript)
