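; Weapon Blueprint for the Lotus Scooter, which marks indicated opponents and shows the mark count on its screen.
; Indicator callbacks apply or remove the marked gameplay effect and update MarkedActors; new marks also send a gameplay event and quest stat.
; Rep notifies push counts into mesh custom primitive data and toggle edge glow on a dynamic screen material created at BeginPlay.
; It also tracks local ammo and leaves debug prints in.
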(event EventBeginPlay
  (bind _weaponmesh (Variables|Visual|GetWeaponMesh))
  (EventDispatchers|AssignOnLocalAmmoChanged (delegate Custom|Ammo%20Changed))
  (|CallParentFunction)
  (for _array_element (Rendering|Material|GetMaterials _weaponmesh)
    (Development|PrintString (Utilities|GetDisplayName _array_element) true true "(R=0.000000,G=0.309195,B=1.000000,A=1.000000)" 4.0))
  (bind _returnvalue (Rendering|Material|CreateDynamicMaterialInstance _weaponmesh 1 (Rendering|Material|GetMaterial _weaponmesh 1) "None" :!class /Script/Engine.PrimitiveComponent))
  (bind _output_get (Variables|Default|SetScreenMaterial _returnvalue))
  (Development|PrintString (Utilities|GetDisplayName _output_get) true true "(R=0.000000,G=0.309195,B=1.000000,A=1.000000)" 4.0))

(event Custom|Ammo%20Changed
  (params (LocalCount :type Integer) (LocalRemaining :type Integer))
  (Development|PrintString (Utilities|String|Append "Scooter New Ammo Amt: " (Utilities|String|ToString(Integer) LocalCount) " Old Ammo Amt: " (Utilities|String|ToString(Integer) (Variables|Default|GetCurrentAmmo))))
  (Variables|Default|SetCurrentAmmo LocalCount))

(fn OnIndicatedActorAdded
  (params (IndicatedActor :type Actor%20Object%20Reference))
  (bind _returnvalue (Game|GetInstigatorController))
  (bind _markedactors (Variables|Default|GetMarkedActors))
  (bind _asfort_pawn (Utilities|Casting|CastToFortPawn IndicatedActor))
  (GameplayEffects|ApplyGameplayEffectToSelf :self (Ability|GetAbilitySystemComponent _asfort_pawn) :GameplayEffectClass (Variables|Default|GetMarkedGE) :Level 0.0)
  (Variables|Default|SetbPlayerAlreadyMarked (Utilities|Array|ContainsItem _markedactors IndicatedActor))
  (Utilities|Array|AddUnique _markedactors IndicatedActor)
  (Rendering|Material|SetCustomPrimitiveDataFloat (Variables|Visual|GetWeaponMesh) (Variables|Default|GetScreenCounterIndex) (Math|Conversions|ToFloat(Integer) (Utilities|Array|Length _markedactors)))
  (if (not (Variables|Default|GetbPlayerAlreadyMarked))
    (bind _asfort_player_controller (Utilities|Casting|CastToFortPlayerController _returnvalue))
    ;[1 1 1] Adding sequence for gameplay event to work with QS2.
    (Utilities|FlowControl|Sequence
      (:then_0
        (Ability|SendGameplayEventtoActor _asfort_player_controller "(TagName=\"Event.MarkedOpponent\")" (Utilities|Struct|MakeGameplayEventData "(TagName=\"\")" _returnvalue _asfort_pawn 0 0 (Utilities|Struct|MakeGameplayEffectContextHandle) "(GameplayTags=)" (GameplayTags|GetOwnedGameplayTags _asfort_pawn :!class /Script/GameplayTags.GameplayTagAssetInterface) 1.0)))
      (:then_1
        (Quest|SendComplexCustomStatEvent (Quests|GetQuestManager _asfort_player_controller "Invalid") 0 "(GameplayTags=((TagName=\"Athena.Quests.Scooter.MarkedPlayer\")))" (GameplayTags|GetOwnedGameplayTags _asfort_pawn :!class /Script/GameplayTags.GameplayTagAssetInterface) 1)))))

(fn OnIndicatedActorRemoved
  (params (IndicatedActor :type Actor%20Object%20Reference))
  (bind _asfort_player_pawn (Utilities|Casting|CastToFortPlayerPawn IndicatedActor))
  (GameplayEffects|RemoveActiveGameplayEffectbySourceEffect (Ability|GetAbilitySystemComponent _asfort_player_pawn) (Variables|Default|GetMarkedGE) 0 -1)
  (Utilities|Array|RemoveItem (Variables|Default|GetMarkedActors) IndicatedActor))

(fn OnRep_MarkedActors
  (bind _markedactors (Variables|Default|GetMarkedActors))
  (bind _returnvalue (Utilities|Array|Length _markedactors))
  (bind _screenmaterial (Variables|Default|GetScreenMaterial))
  (Rendering|Material|SetCustomPrimitiveDataFloat (Variables|Visual|GetWeaponMesh) (Variables|Default|GetScreenCounterIndex) (Math|Conversions|ToFloat(Integer) _returnvalue))
  (if (> _returnvalue 0)
    (Rendering|Material|SetScalarParameterValue _screenmaterial "EdgeBrightnessScalar" 0.3 :!class /Script/Engine.MaterialInstanceDynamic)
    (else
      (Rendering|Material|SetScalarParameterValue _screenmaterial "EdgeBrightnessScalar" 0.0 :!class /Script/Engine.MaterialInstanceDynamic)))
  (Development|PrintString "UPDATE SCREEN"))

(fn OnRep_MarkedPlayerCount
  (if (not (Networking|IsServer))
    (Rendering|Material|SetCustomPrimitiveDataFloat (Variables|Visual|GetWeaponMesh) 0 (Math|Conversions|ToFloat(Integer) (Variables|Default|GetMarkedPlayerCount)))))

(fn ConstructionScript
  (|CallParentFunction))
