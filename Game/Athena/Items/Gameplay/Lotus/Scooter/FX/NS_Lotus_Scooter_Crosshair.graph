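# Niagara system for the Lotus Scooter crosshair, drawn as a glowing red sprite.
# A single CrossHair emitter burst-spawns sprites sized by the owner component's scale, placed with the world-space User.WorldOffset and nudged by a camera offset.
# Curves over particle life pop the size up then shrink it, and quickly fade alpha in then out, while a high RGB multiplier brightens the color.
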NiagaraSystem /Game/Athena/Items/Gameplay/Lotus/Scooter/FX/NS_Lotus_Scooter_Crosshair
dependencies {usedDataInterfaces [&NiagaraDataInterfaceCurve], usedDynamicInputs [], usedModules [&SystemState]}
emitters {CrossHair {
  emitterUpdateScript [
   &EmitterState()
   &SpawnBurst_Instantaneous(Use Spawn Probability=false, UseLoopCountLimit=false)
  ]
  particleSpawnScript [
   &InitializeParticle(
    Color=<rgba 0.75 0.051 0.051 2>
    Position Offset=$User.WorldOffset
    Position Offset Coordinate Space=::World
    Sprite Rotation Mode=::NewEnumerator1
    Sprite Size=&Multiply_Vector2D_ByFloat(Float=&MakeFloatFromVector(Vector=$Engine.Owner.Scale, Channel=::NewEnumerator0), Vector2D=<100 100>)
    Sprite Size Mode=::NewEnumerator4
    UsePositionOffset=true
    Write Material Random=true
   )
   &CameraOffset(Camera Offset Amount=8)
  ]
  particleUpdateScript [
   &ParticleState()
   &ScaleSpriteSize(
    Scale Factor=&Vector2DFromFloat(Value=&FloatFromCurve(FloatCurve=DataInterface() {propertyValues {Curve {keys [
          <at 0 0.4 RCIM_Cubic>
          {arriveTangent=0.3438, interpMode=RCIM_Cubic, leaveTangent=0.3438, time=0.1158, value=1.0077}
          {arriveTangent=-0.5366, interpMode=RCIM_Cubic, leaveTangent=-0.5366, time=0.5736, value=0.5972}
          <at 1.0021 0.5322 RCIM_Cubic>
         ]}}}))
    Scale Sprite Size Mode=::NewEnumerator2
   )
   &ScaleColor(
    Scale Alpha=&FloatFromCurve(FloatCurve=DataInterface() {propertyValues {Curve {keys [
         <at 0 0 RCIM_Cubic>
         {arriveTangent=3.2341, interpMode=RCIM_Cubic, leaveTangent=3.883, tangentMode=RCTM_Break, time=0.1709, value=0.175}
         {arriveTangent=-3.0279, interpMode=RCIM_Cubic, leaveTangent=-3.0279, tangentMode=RCTM_User, time=0.2586, value=1.0042}
         <at 1 0>
        ]}}})
    Scale RGB=<25 25 25>
    ScaleA=true
    ScaleRGB=true
   )
  ]
  rendererClasses [&NiagaraSpriteRendererProperties]
  renderers [{rendererClass=&NiagaraSpriteRendererProperties}]
 }}
summary {
 systemName=NS_Lotus_Scooter_Crosshair
 emitters [{emitterName=CrossHair, rendererClasses [&NiagaraSpriteRendererProperties]}]
 userVariables [
  User.Vector_Vector {type=Vector3f, description="", defaultValue {struct=Vector3f, value=<0 0 0>}}
  User.WorldOffset {type=Vector3f, description="", defaultValue {struct=Vector3f, value=<0 0 0>}}
 ]
}
