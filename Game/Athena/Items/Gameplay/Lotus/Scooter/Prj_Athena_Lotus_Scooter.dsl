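; Lotus Scooter grenade-based projectile that periodically marks nearby actors with indicators and stencils for the instigator's team.
; On BeginPlay it applies team-colored FX and mesh materials, then starts PulseIndicatorComponent if usable, otherwise loops a PulseMarkers timer.
; PulseMarkers rebuilds actor filters with scalable radius and duration, adding particles and ping sound only for the "players" group.
; Add/remove callbacks bind to the owning weapon, and EndPlay stops the component's pulses.
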(event EventBeginPlay
  (bind _pulseindicator (Variables|Default|GetPulseIndicator))
  ;[1 1 1] BeginPlay - Set Team Color
  (Utilities|FlowControl|Sequence
    (:then_0
      (Networking|CanExecuteCosmeticEvents
        (:True
          (Components|Activation|Activate (Variables|Default|GetTeamColored_FX) false)
          (Effects|Components|ParticleSystem|SetColorParameter (Variables|Default|GetTeamColored_FX) "TeamColor" (Utilities|GetTeamVFX (Variables|Default|GetBeacon%20Streak%20Color%20%28Friendly%29) (Variables|Default|GetBeacon%20Streak%20Color%20%28Enemy%29))) ; Set team color.
          (Rendering|Material|SetMaterial (Variables|B_Prj_Grenade|GetMesh) 0 (Selection|SelectMaterialInstance (Variables|Default|GetFriendlyMeshMaterial) (Variables|Default|GetEnemyMeshMaterial) (Team|OnSameTeam self (Game|GetPlayerController 0))))
          ;;[0.6458 0.6458 0.6458] Set Material based on Team
          (if (not (Variables|Default|GetIsDead?))
            (Components|Activation|Activate (Variables|Default|GetTeamColored_FX) false)
            (Utilities|FlowControl|Delay 0.15)))))
    (:then_1
      (bind _asathena_player_controller (Utilities|Casting|CastToAthena_PlayerController (Fortnite|GetFortPlayerControllerfromActor (Variables|Actor|GetInstigator))))
      (Variables|Default|SetAthenaPlayerControler _asathena_player_controller))
    (:then_2
      (if (Class|PulseIndicatorComponent|CanUseComponentIndicators (Variables|Default|GetPulseIndicator))
        (Class|PulseIndicatorComponent|SetDisplayTeam _pulseindicator (CallFunction|GetMarkerDisplayTeam))
        (Class|PulseIndicatorComponent|SetIndicatingController _pulseindicator (Variables|Default|GetAthenaPlayerControler))
        (bind _returnvalue (CallFunction|GetOwnerWeapon))
        (Variables|/Script/FortniteGame.PulseIndicatorComponent|SetOnActorAdded (EventDispatchers|CreateEvent (Utilities|Casting|CastToB_Athena_Lotus_Scooter_Weap _returnvalue) :!event OnIndicatedActorAdded) _pulseindicator)
        (Variables|/Script/FortniteGame.PulseIndicatorComponent|SetOnActorRemoved (EventDispatchers|CreateEvent (Utilities|Casting|CastToB_Athena_Lotus_Scooter_Weap _returnvalue) :!event OnIndicatedActorRemoved) _pulseindicator)
        (Class|PulseIndicatorComponent|StartPulseMarkers _pulseindicator)
        (else
          (CallFunction|PulseMarkers)
          (bind _returnvalue_2 (Utilities|Time|SetTimerbyFunctionName 0 "PulseMarkers" (Variables|Default|GetPulseTime) true))
          (Variables|Default|SetPulseTimer _returnvalue_2))))))

(event EventEndPlay
  (params (EndPlayReason :type EEndPlayReason%20Enum :default "Destroyed"))
  ;[1 1 1] BeginPlay - Set Team Color
  (|CallParentFunction EndPlayReason)
  (Class|PulseIndicatorComponent|StopPulseMarkers (Variables|Default|GetPulseIndicator)))

(fn CreateModifiedIndicatedActorFilter
  (params (NewParam :type Indicated%20Actor%20Data%20With%20Filter%20Structure))
  (bind (_indicatedactortags _indicatedactortagquery _objecttypes _actorclassfilter _withaffiliation _indicateddata _stencileddata) (Utilities|Struct|BreakIndicatedActorDatawithFilter NewParam))
  (bind (_groupidentifier _duration _steptime _shareactorwith _shareactorwithmask _bvisibletoowner _displayteamoverride _particlesystemdata _sound _friendlystencilname _enemystencilname) (Utilities|Struct|BreakStenciledActorData _stencileddata))
  (bind (_groupidentifier_2 _duration_2 _steptime_2 _indicatoroffset _indicatordbnooffset _shareactorwith_2 _shareactorwithmask_2 _bvisibletoowner_2 _displayteamoverride_2 _scaleandopacitydata _particlesystemdata_2 _sound_2 _stateimageoverride) (Utilities|Struct|BreakIndicatedActorData _indicateddata))
  (bind _returnvalue (CallFunction|GetOwnerWeapon))
  (return (Utilities|Struct|MakeIndicatedActorDatawithFilter _indicatedactortags _indicatedactortagquery _objecttypes _actorclassfilter _withaffiliation (Utilities|Struct|MakeIndicatedActorData "LotusIndicator" (ScalableFloat|GetValueatLevel (Variables|Default|GetIndividualMarkDuration)) (Variables|Default|GetStep%20Time) "0.000000,0.000000,90.000000" "0.000000,0.000000,30.000000" "None" _shareactorwithmask_2 true (CallFunction|GetMarkerDisplayTeam) "(SmallSizeDistance=7500.000000,LargestSizeDistance=10.000000,SmallScale=0.500000,LargestScale=0.800000,FarAwayScale=0.350000,FarAwayOpacity=1.000000,MaxScaleAndFadePercent=0.100000)" (select (Utilities|String|EqualExactly(String) _groupidentifier "players") (IndicatedManagement|MakeIndicatedActorParticleSystemData (Variables|Default|GetIndicatedPS) (Variables|Default|GetParticleSystemParamName) (Variables|Default|GetPSOffset) (Variables|Default|GetPSDBNOOffset) (Variables|Default|GetPSVectorParamName)) 0) (select (Utilities|String|EqualExactly(String) _groupidentifier "players") (Variables|Default|GetPingSound) 0) _stateimageoverride -1 true false false false (EventDispatchers|CreateEvent (Utilities|Casting|CastToB_Athena_Lotus_Scooter_Weap _returnvalue) :!event OnIndicatedActorAdded) (EventDispatchers|CreateEvent (Utilities|Casting|CastToB_Athena_Lotus_Scooter_Weap _returnvalue) :!event OnIndicatedActorRemoved)) (Utilities|Struct|MakeStenciledActorData "LotusStencil" (ScalableFloat|GetValueatLevel (Variables|Default|GetIndividualMarkDuration)) (Variables|Default|GetStep%20Time) _shareactorwithmask true (CallFunction|GetMarkerDisplayTeam) "(ParticleSystem=None,ParticleSystemActorParamName=\"\",ParticleSystemOffset=(X=0.000000,Y=0.000000,Z=0.000000),ParticleSystemDBNOOffset=(X=0.000000,Y=0.000000,Z=0.000000),ParticleSystemVectorParamName=\"\")" 0 _friendlystencilname _enemystencilname) (ScalableFloat|GetValueatLevel (Variables|Default|GetMarkRadius)))))

(fn GetMarkerDisplayTeam
  (bind _returnvalue (CallFunction|GetOwnerWeapon))
  (Utilities|IsValid _returnvalue
    (:Is%20Valid
      (Utilities|IsValid (Game|GetInstigator _returnvalue)
        (:Is%20Valid
          (return (Team|GetActorTeam (Game|GetInstigator _returnvalue))))
        (:Is%20Not%20Valid
          _)))
    (:Is%20Not%20Valid
      _))
  (return 255))

(fn OnActorAdded
  (params (IndicatedActor :type Actor%20Object%20Reference))
  (Utilities|Array|AddUnique (Variables|Default|GetMarkedPlayers) IndicatedActor)
  (bind _returnvalue (CallFunction|GetOwnerWeapon))
  (bind _asb_athena_lotus_scooter_weap (Utilities|Casting|CastToB_Athena_Lotus_Scooter_Weap _returnvalue))
  (Utilities|Array|AddUnique (Variables|/Game/Athena/Items/Gameplay/Lotus/Scooter/B_Athena_Lotus_Scooter_Weap.B_Athena_Lotus_Scooter_Weap_C|GetMarkedActors _asb_athena_lotus_scooter_weap) IndicatedActor))

(fn OnActorRemoved
  (params (IndicatedActor :type Actor%20Object%20Reference))
  (Utilities|Array|RemoveItem (Variables|Default|GetMarkedPlayers) IndicatedActor)
  (bind _returnvalue (CallFunction|GetOwnerWeapon))
  (bind _asb_athena_lotus_scooter_weap (Utilities|Casting|CastToB_Athena_Lotus_Scooter_Weap _returnvalue))
  (Utilities|Array|RemoveItem (Variables|/Game/Athena/Items/Gameplay/Lotus/Scooter/B_Athena_Lotus_Scooter_Weap.B_Athena_Lotus_Scooter_Weap_C|GetMarkedActors _asb_athena_lotus_scooter_weap) IndicatedActor))

(fn PulseMarkers
  (for _array_element (Variables|Default|GetIndicatedActorFilters)
    (Utilities|Array|Add (LocalVariables|GetActorFilters) (CallFunction|CreateModifiedIndicatedActorFilter _array_element)))
  (IndicatedManagement|AddActorsInRadiusToIndicatedList (Variables|Default|GetAthenaPlayerControler) (LocalVariables|GetActorFilters) false false true self))

(fn ConstructionScript
  (Networking|CanExecuteCosmeticEvents
    (:True
      (Components|Movement|SetUpdatedComponent (Variables|Default|GetRotatingMovement) (Variables|B_Prj_Grenade|GetMesh)))))
